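# Physics asset for the FrostGecko vehicle body cosmetic mesh, giving each named part a single capsule collision shape.
# Large overlapping capsules cover body, trim, lenses and wheels, while smaller ones cover the front, side and opaque glass.
# Every part constrains directly to the tiny FrostGecko root capsule in a flat star, with identical swing and twist limit angles.
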PhysicsAsset /VehicleCosmetics/Bodies/FrostGecko/Meshes/SM_FrostGecko_LOD0_PhysicsAsset
bodies {
 Body_01 [{shapeType=Capsule, length=231.1266, radius=119.2775, center=<29.1827 -90.3649 0>, rotation=<rot 0 -81.9315 -90>}]
 FrostGecko [{shapeType=Capsule, length=0.505, radius=0.505, rotation=<rot 0 -90 0>}]
 Glass_F [{shapeType=Capsule, length=79.8266, radius=40.7719, center=<108.5503 -126.3827 0>, rotation=<rot 0 -90 0>}]
 Glass_L_B [{shapeType=Capsule, length=30.5424, radius=18.1761, center=<-53.7249 -134.9096 -81.8033>, rotation=<rot 0 -78.7233 -82.644>}]
 Glass_L_F [{shapeType=Capsule, length=56.9254, radius=19.6012, center=<36.2849 -125.4703 -85.9212>, rotation=<rot 0 -80.6063 -87.4762>}]
 Glass_Opaque [{shapeType=Capsule, length=72.4348, radius=14.1976, center=<-155.8006 -134.625 0>, rotation=<rot 0 -90 0>}]
 Glass_R_B [{shapeType=Capsule, length=30.5424, radius=18.1761, center=<-53.7249 -134.9096 81.8033>, rotation=<rot 0 101.2767 -82.644>}]
 Glass_R_F [{shapeType=Capsule, length=56.9254, radius=19.6012, center=<36.2849 -125.4703 85.9212>, rotation=<rot 0 99.3937 -87.4762>}]
 Lenses_01 [{shapeType=Capsule, length=211.7301, radius=118.5405, center=<37.3682 -102.5446 0>, rotation=<rot 0 95.4481 -90>}]
 SK_Wheel_OEM [{shapeType=Capsule, length=200.3607, radius=169.3613, center=<7.8362 -37.9024 0>, rotation=<rot 0 -90 -110.4922>}]
 Trim_01 [{shapeType=Capsule, length=241.7694, radius=118.0282, center=<36.3408 -81.3988 -0.0054>, rotation=<rot 0 -84.8619 -89.9964>}]
}
constraints [
 {bone1Name=SK_Wheel_OEM, bone2Name=FrostGecko, swing1LimitDegrees=45, swing2LimitDegrees=45, twistLimitDegrees=45}
 {bone1Name=Lenses_01, bone2Name=FrostGecko, swing1LimitDegrees=45, swing2LimitDegrees=45, twistLimitDegrees=45}
 {bone1Name=Glass_Opaque, bone2Name=FrostGecko, swing1LimitDegrees=45, swing2LimitDegrees=45, twistLimitDegrees=45}
 {bone1Name=Glass_F, bone2Name=FrostGecko, swing1LimitDegrees=45, swing2LimitDegrees=45, twistLimitDegrees=45}
 {bone1Name=Glass_L_F, bone2Name=FrostGecko, swing1LimitDegrees=45, swing2LimitDegrees=45, twistLimitDegrees=45}
 {bone1Name=Glass_L_B, bone2Name=FrostGecko, swing1LimitDegrees=45, swing2LimitDegrees=45, twistLimitDegrees=45}
 {bone1Name=Glass_R_F, bone2Name=FrostGecko, swing1LimitDegrees=45, swing2LimitDegrees=45, twistLimitDegrees=45}
 {bone1Name=Glass_R_B, bone2Name=FrostGecko, swing1LimitDegrees=45, swing2LimitDegrees=45, twistLimitDegrees=45}
 {bone1Name=Trim_01, bone2Name=FrostGecko, swing1LimitDegrees=45, swing2LimitDegrees=45, twistLimitDegrees=45}
 {bone1Name=Body_01, bone2Name=FrostGecko, swing1LimitDegrees=45, swing2LimitDegrees=45, twistLimitDegrees=45}
]
